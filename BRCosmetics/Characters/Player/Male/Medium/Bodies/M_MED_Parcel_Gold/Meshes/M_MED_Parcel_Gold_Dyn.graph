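# Physics asset for the dynamic front and back loincloth on the male medium Parcel Gold player cosmetic body.
# Each flap is a four-box bone chain with limited swing and tight twist, rooted at a tiny MSTR capsule constrained to the pelvis.
# Capsules on pelvis and thighs give the lower-body shapes around the flaps.
# Constraints linking thighs and MSTR roots to the pelvis leave swing and twist free.
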PhysicsAsset /BRCosmetics/Characters/Player/Male/Medium/Bodies/M_MED_Parcel_Gold/Meshes/M_MED_Parcel_Gold_Dyn
bodies {
 dyn_loincloth_bk_01 [{shapeType=Box, extentX=12.4375, extentY=14.9375, extentZ=1.625, center=<6.118 0 0.0001>}]
 dyn_loincloth_bk_01_MSTR [{shapeType=Capsule, length=0.505, radius=0.505, rotation=<rot 0 -90 0>}]
 dyn_loincloth_bk_02 [{shapeType=Box, extentX=10.4375, extentY=16.8125, extentZ=1.625, center=<5.2484 0 0>}]
 dyn_loincloth_bk_03 [{shapeType=Box, extentX=9.4375, extentY=18.1875, extentZ=1.625, center=<4.6445 0 0>}]
 dyn_loincloth_bk_04 [{shapeType=Box, extentX=7.5, extentY=18.375, extentZ=1.625, center=<4.2148 0 0>}]
 dyn_loincloth_fr_01 [{shapeType=Box, extentX=7.5, extentY=11.25, extentZ=1.625, center=<4.2148 0 0>}]
 dyn_loincloth_fr_01_MSTR [{shapeType=Capsule, length=0.505, radius=0.505, rotation=<rot 0 -90 0>}]
 dyn_loincloth_fr_02 [{shapeType=Box, extentX=8.875, extentY=11.875, extentZ=1.625, center=<4.2148 0 0>}]
 dyn_loincloth_fr_03 [{shapeType=Box, extentX=8.875, extentY=12.5625, extentZ=1.625, center=<4.2148 0 0>}]
 dyn_loincloth_fr_04 [{shapeType=Box, extentX=8.3125, extentY=13.3125, extentZ=1.625, center=<4.2148 0 0>}]
 pelvis [{shapeType=Capsule, length=17.0987, radius=9.844, center=<-2.9744 0.8635 0.0001>}]
 thigh_l [{shapeType=Capsule, length=24.0674, radius=8.8128, center=<-16.1146 1.0428 -1.4837>, rotation=<rot 90 0 0>}]
 thigh_r [{shapeType=Capsule, length=24.0674, radius=8.8128, center=<16.1146 -1.0428 1.4837>, rotation=<rot 90 0 0>}]
}
constraints [
 {bone1Name=dyn_loincloth_bk_01, bone2Name=dyn_loincloth_bk_01_MSTR, swing1LimitDegrees=30, swing2LimitDegrees=45, twistLimitDegrees=2}
 {bone1Name=dyn_loincloth_bk_02, bone2Name=dyn_loincloth_bk_01, swing1LimitDegrees=45, swing2LimitDegrees=45, twistLimitDegrees=5}
 {bone1Name=dyn_loincloth_bk_03, bone2Name=dyn_loincloth_bk_02, swing1LimitDegrees=45, swing2LimitDegrees=45, twistLimitDegrees=5}
 {bone1Name=dyn_loincloth_bk_04, bone2Name=dyn_loincloth_bk_03, swing1LimitDegrees=45, swing2LimitDegrees=45, twistLimitDegrees=5}
 {bone1Name=dyn_loincloth_fr_01, bone2Name=dyn_loincloth_fr_01_MSTR, swing1LimitDegrees=45, swing2LimitDegrees=45, twistLimitDegrees=5}
 {bone1Name=dyn_loincloth_fr_02, bone2Name=dyn_loincloth_fr_01, swing1LimitDegrees=45, swing2LimitDegrees=45, twistLimitDegrees=5}
 {bone1Name=dyn_loincloth_fr_03, bone2Name=dyn_loincloth_fr_02, swing1LimitDegrees=45, swing2LimitDegrees=45, twistLimitDegrees=5}
 {bone1Name=dyn_loincloth_fr_04, bone2Name=dyn_loincloth_fr_03, swing1LimitDegrees=45, swing2LimitDegrees=45, twistLimitDegrees=5}
 {
  bone1Name=thigh_l
  bone2Name=pelvis
  swing1LimitDegrees=45
  swing1Motion=Free
  swing2LimitDegrees=45
  swing2Motion=Free
  twistLimitDegrees=45
  twistMotion=Free
 }
 {
  bone1Name=thigh_r
  bone2Name=pelvis
  swing1LimitDegrees=45
  swing1Motion=Free
  swing2LimitDegrees=45
  swing2Motion=Free
  twistLimitDegrees=45
  twistMotion=Free
 }
 {
  bone1Name=dyn_loincloth_bk_01_MSTR
  bone2Name=pelvis
  swing1LimitDegrees=45
  swing1Motion=Free
  swing2LimitDegrees=45
  swing2Motion=Free
  twistLimitDegrees=45
  twistMotion=Free
 }
 {
  bone1Name=dyn_loincloth_fr_01_MSTR
  bone2Name=pelvis
  swing1LimitDegrees=45
  swing1Motion=Free
  swing2LimitDegrees=45
  swing2Motion=Free
  twistLimitDegrees=45
  twistMotion=Free
 }
]
