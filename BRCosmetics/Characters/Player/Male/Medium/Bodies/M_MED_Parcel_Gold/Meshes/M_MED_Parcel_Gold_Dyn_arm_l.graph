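# Physics asset for the left-arm dynamic piece of the gold Parcel medium male player body, covering its elbow tube.
# Two capsule bodies sit on dyn_elbow_tube_l and dyn_elbow_tube_l_MSTR, the latter smaller and rotated a quarter turn.
# A single constraint joins the tube bone to its MSTR bone with equal swing and twist angle limits.
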PhysicsAsset /BRCosmetics/Characters/Player/Male/Medium/Bodies/M_MED_Parcel_Gold/Meshes/M_MED_Parcel_Gold_Dyn_arm_l
bodies {
 dyn_elbow_tube_l [{shapeType=Capsule, length=0.5312, radius=1.6875}]
 dyn_elbow_tube_l_MSTR [{shapeType=Capsule, length=0.505, radius=0.505, rotation=<rot 0 -90 0>}]
}
constraints [{bone1Name=dyn_elbow_tube_l, bone2Name=dyn_elbow_tube_l_MSTR, swing1LimitDegrees=45, swing2LimitDegrees=45, twistLimitDegrees=45}]
