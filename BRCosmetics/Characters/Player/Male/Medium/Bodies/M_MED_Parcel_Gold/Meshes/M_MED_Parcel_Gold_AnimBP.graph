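# Secondary-motion AnimBP for the Parcel Gold medium male body cosmetic, layered on the pose copied from the attached parent mesh.
# Weighted constraints drive elbow tubes and pistols, then look-at and additive roll adjust shoulder pads, partly weighted by upper-arm twist computed each update.
# Three RigidBody passes then simulate dynamics relative to the spine and each upper arm.
# A final bone transform offsets and squashes the front loincloth.
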AnimBlueprint /BRCosmetics/Characters/Player/Male/Medium/Bodies/M_MED_Parcel_Gold/Meshes/M_MED_Parcel_Gold_AnimBP
target_skeleton=&/M_MED_Parcel_Gold_Skeleton
graphs [
 AnimGraph {nodes [
   n1 = "Constraint - dyn_elbow_tube_l_MSTR (PPP)"!(
    ComponentPose=Local To Component(LocalPose=Copy Pose From Mesh() {bUseAttachedParent=true})
    Alpha=1
    ConstraintWeights_0=0.0152
    ConstraintWeights_1:float=0.9146
    ConstraintWeights_2:float=0.0702
   ) -> (Pose) {
    alphaBoolBlend {blendOption=Linear}
    boneToModify {boneName=dyn_elbow_tube_l_MSTR}
    constraintSetup [{targetBone {boneName=upperarm_twist_02_l}}, {targetBone {boneName=upperarm_l}}, {targetBone {boneName=lowerarm_twist_02_l}}]
    constraintWeights [1,1,1]
   }
   n2 = "Constraint - dyn_elbow_tube_r_MSTR (PPP)"!(ComponentPose=@n1, Alpha=1, ConstraintWeights_0=0.0702, ConstraintWeights_1:float=0.9146, ConstraintWeights_2:float=0.0152) -> (Pose) {
    alphaBoolBlend {blendOption=Linear}
    boneToModify {boneName=dyn_elbow_tube_r_MSTR}
    constraintSetup [{targetBone {boneName=lowerarm_twist_02_r}}, {targetBone {boneName=upperarm_r}}, {targetBone {boneName=upperarm_twist_02_r}}]
    constraintWeights [1,1,1]
   }
   n3 = "Constraint - dyn_pistol_l (PPP)"!(ComponentPose=@n2, Alpha=1, ConstraintWeights_0=0.2538, ConstraintWeights_1:float=0.731, ConstraintWeights_2:float=0.0152) -> (Pose) {
    alphaBoolBlend {blendOption=Linear}
    boneToModify {boneName=dyn_pistol_l}
    constraintSetup [{targetBone {boneName=pelvis}}, {targetBone {boneName=thigh_twist_01_l}}, {targetBone {boneName=spine_01}}]
    constraintWeights [1,1,1]
   }
   n4 = "Constraint - dyn_pistol_r (PPP)"!(ComponentPose=@n3, Alpha=1, ConstraintWeights_0=0.2528, ConstraintWeights_1:float=0.7326, ConstraintWeights_2:float=0.0146) -> (Pose) {
    alphaBoolBlend {blendOption=Linear}
    boneToModify {boneName=dyn_pistol_r}
    constraintSetup [{targetBone {boneName=pelvis}}, {targetBone {boneName=thigh_twist_01_r}}, {targetBone {boneName=spine_01}}]
    constraintWeights [1,1,1]
   }
   n5 = Get shoulder l rot() -> (shoulder l rot:double)
   n6 = Look At - Bone: dyn_shoulderpad_l_MSTR(ComponentPose=@n4, Alpha=@n5) {
    alphaBoolBlend {blendOption=Linear}
    boneToModify {boneName=dyn_shoulderpad_l_MSTR}
    lookAtLocation {x=0}
    lookAtTarget {boneReference {boneName=lowerarm_l}}
    lookAt_Axis {axis=<1 0>}
    lookUp_Axis {axis {x=0, z=1}}
   }
   n7 = "Transform (Modify) Bone - Bone: dyn_shoulderpad_l_MSTR"(ComponentPose=@n6, Alpha=@n5, Translation="0,0,0", Rotation="0,0.000,-30", Scale="1,1,1") {
    rotationMode=BMM_Additive
    rotationSpace=BCS_ParentBoneSpace
    alphaBoolBlend {blendOption=Linear}
    boneToModify {boneName=dyn_shoulderpad_l_MSTR}
    rotation {roll=-30}
   }
   n8 = Look At - Bone: dyn_shoulderpad_r_MSTR(ComponentPose=@n7, Alpha=1) {
    alphaBoolBlend {blendOption=Linear}
    boneToModify {boneName=dyn_shoulderpad_r_MSTR}
    lookAtLocation {x=0}
    lookAtTarget {boneReference {boneName=lowerarm_r}}
    lookAt_Axis {axis=<1 0>}
    lookUp_Axis {axis {x=0, z=1}}
   }
   n9 = "Transform (Modify) Bone - Bone: dyn_shoulderpad_r_MSTR"(ComponentPose=@n8, Alpha=Get shoulder r rot() -> (shoulder r rot:double), Translation="0,0,0", Rotation="0,0.000,30", Scale="1,1,1") {
    rotationMode=BMM_Additive
    rotationSpace=BCS_BoneSpace
    alphaBoolBlend {blendOption=Linear}
    boneToModify {boneName=dyn_shoulderpad_r_MSTR}
    rotation {roll=30}
   }
   n10 = RigidBody(ComponentPose=@n9) {
    bClampLinearTranslationLimitToRefPose=true
    bOverrideWorldGravity=true
    overridePhysicsAsset=&/M_MED_Parcel_Gold_Dyn
    simulationSpace=BaseBoneSpace
    alphaBoolBlend {blendOption=Linear}
    baseBoneRef {boneName=spine_01}
    componentAppliedLinearAccClamp=<3000 3000 3000>
    componentLinearAccScale=<0.4 0.3 0.3>
    componentLinearVelScale=<0.4 0.3 0.3>
    overrideWorldGravity {z=-980}
   }
   n11 = RigidBody(ComponentPose=@n10) {
    bClampLinearTranslationLimitToRefPose=true
    bOverrideWorldGravity=true
    overridePhysicsAsset=&/M_MED_Parcel_Gold_Dyn_arm_l
    simulationSpace=BaseBoneSpace
    alphaBoolBlend {blendOption=Linear}
    baseBoneRef {boneName=upperarm_l}
    componentAppliedLinearAccClamp=<3000 3000 3000>
    componentLinearAccScale=<0.6 0.6 0.6>
    componentLinearVelScale=<0.6 0.6 0.6>
    overrideWorldGravity {z=-980}
   }
   n12 = RigidBody(ComponentPose=@n11) {
    bClampLinearTranslationLimitToRefPose=true
    bOverrideWorldGravity=true
    overridePhysicsAsset=&/M_MED_Parcel_Gold_Dyn_arm_r
    simulationSpace=BaseBoneSpace
    alphaBoolBlend {blendOption=Linear}
    baseBoneRef {boneName=upperarm_r}
    componentAppliedLinearAccClamp=<3000 3000 3000>
    componentLinearAccScale=<0.6 0.6 0.6>
    componentLinearVelScale=<0.6 0.6 0.6>
    overrideWorldGravity {z=-980}
   }
   n13 = "Transform (Modify) Bone - Bone: dyn_loincloth_fr_01_MSTR"!(ComponentPose=@n12, Alpha=1, Translation="-1.5,0,1", Scale="1,0.75,1") -> (Pose) {
    rotationMode=BMM_Replace
    rotationSpace=BCS_BoneSpace
    scaleMode=BMM_Replace
    scaleSpace=BCS_BoneSpace
    translationMode=BMM_Replace
    translationSpace=BCS_BoneSpace
    alphaBoolBlend {blendOption=Linear}
    alphaScaleBiasClamp {bInterpResult=true, interpSpeedDecreasing=20, interpSpeedIncreasing=5}
    boneToModify {boneName=dyn_loincloth_fr_01_MSTR}
    scale {y=0.75}
    translation {x=-1.5, z=1}
   }
   Output Pose(Result=Component To Local(ComponentPose=@n13)) {layerGroup=None}
  ]}
 EventGraph {nodes [
   Event BlueprintUpdateAnimation()
   n1 = GetPartSkeletalMeshComponent()
   > Set shoulder l rot(shoulder l rot:double=GetAlphaFromReferenceTransformDelta(
     SkeletalMeshComponent=@n1
     BoneName=upperarm_l
     BaseName=clavicle_l
     DeltaComponentType=QuaternionTwist
     InputRange=(X=-10,Y=-100)
     TwistAxis="1,0,0"
    )) -> (Output_Get:double)
   > Set shoulder r rot(shoulder r rot:double=GetAlphaFromReferenceTransformDelta(
     SkeletalMeshComponent=@n1
     BoneName=upperarm_r
     BaseName=clavicle_r
     DeltaComponentType=QuaternionTwist
     InputRange=(X=-10,Y=-100)
     TwistAxis="1,0,0"
    )) -> (Output_Get:double)
  ]}
]
variables [
 shoulder l rot {type=number}
 shoulder r rot {type=number}
]
